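; Camera lens Blueprint for the Galileo rocket glider's deploy effect, driving a post-process component named Post_Space.
; On BeginPlay it calls the parent, then plays the Post Controller timeline from the start.
; Each timeline update sets the post-process blend weight from the curve value, animating the effect's strength over time.
; The construction script is empty.
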(event EventBeginPlay
  (|CallParentFunction)
  (bind (_direction _lerp) (Timeline|Post%20Controller PlayFromStart)
    (:Update
      (Variables|/Script/Engine.PostProcessComponent|SetBlendWeight _lerp (Variables|Default|GetPost_Space)))))

(fn ConstructionScript)
